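# Looping spot ambience of liquid metal flowing inside the Nitro Silo at the Helios NitroEngine POI.
# A looping wave passes through a Modulator into an endlessly looping Enveloper that slowly raises pitch while dipping volume, then returns, disguising repetition.
# The Quality Level root routes this chain to its first two levels, leaving two empty.
# Overrides add capsule attenuation with listener focus and tight concurrency; a bus send feeds exterior ambience, and the sound class is Fort_Doors.
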SoundCue /Game/Sounds/Helios/POI/NitroEngine/Spot/Spot_NitroSilo_Liquid_Metal_Flowing_Inside_Loop_Cue
root=QualityLevel_0
nodes [
 {
  id=Enveloper_0
  children [Modulator_2]
  properties {
   bLoopIndefinitely=true
   loopEnd=30
   pitchCurve {editorCurveData {keys [{interpMode=RCIM_Cubic, time=0}, <at 15 1.1 RCIM_Cubic>, {interpMode=RCIM_Cubic, time=30}]}}
   volumeCurve {editorCurveData {keys [<at 0 1 RCIM_Cubic>,<at 15 0.7 RCIM_Cubic>,<at 30 1 RCIM_Cubic>]}}
  }
 }
 {id=Modulator_2, children [WavePlayer_0]}
 {id=QualityLevel_0, children [Enveloper_0,Enveloper_0,null,null]}
 {
  id=WavePlayer_0
  properties {bLooping=true, soundWaveAssetPtr=&/Spot_NitroSilo_Liquid_Metal_Flowing_Inside_Loop}
 }
]
properties {
 bEnableSubmixSends=false
 bOverrideAttenuation=true
 bOverrideConcurrency=true
 maxDistance=1913.5496
 soundClassObject=&Fort_Doors
 volumeMultiplier=0.25
 attenuationOverrides {
  attenuationShape=Capsule
  bEnableFocusInterpolation=true
  bEnableListenerFocus=true
  falloffDistance=763.7398
  focusAzimuth=25
  focusReleaseInterpSpeed=0.9
  nonFocusAzimuth=50
  nonFocusVolumeAttenuation=0
  nonSpatializedRadiusEnd=125
  nonSpatializedRadiusStart=150
  stereoSpread=150
  attenuationShapeExtents=<1149.8098 239.9017>
 }
 busSends [{audioBus=null, minSendLevel=0.6, sendLevel=1, soundSourceBus=&AmbientAudio_Exterior_SourceBus}]
 concurrencyOverrides {maxCount=2, voiceStealReleaseTime=1}
 concurrencySet [&Weapon_Drop_StopOldest]
}
